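; Treasure map weapon for the Search For Booty secondary mission that shows booty beacons to the player holding it.
; On equip and attach, it validates the instigator, sends it the Booty Equipped gameplay event, plays a sound at its location and calls ShowBeacon on every SFB_Booty actor.
; Local unequip mirrors this with the Unequipped event, another sound and HideBeacon.
; Booty actors are found by a class-wide search on each change.
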(event FortWeapon|EventOnEquippedAndAttached
  (bind _returnvalue (Game|GetInstigator))
  (Utilities|IsValid _returnvalue
    (:Is%20Valid
      (Ability|SendGameplayEventtoActor _returnvalue "(TagName=\"Event.Athena.Booty.Equipped\")")
      (Audio|PlaySoundatLocation "/Game/Sounds/Foley_Loot/Containers/Pirate_Booty/Pirate_Booty_Equip_Cue.Pirate_Booty_Equip_Cue" (Transformation|GetActorLocation _returnvalue))
      (bind _outactors (Actor|GetAllActorsOfClass "/Game/Missions/Secondary/SearchForBooty/Props/SFB_Booty.SFB_Booty_C"))
      (for _array_element _outactors
        (Class|SFBBooty|ShowBeacon _array_element _returnvalue)))))

(event FortWeapon|EventOnLocalUnEquip
  (bind _returnvalue (Game|GetInstigator))
  (Utilities|IsValid _returnvalue
    (:Is%20Valid
      (Ability|SendGameplayEventtoActor _returnvalue "(TagName=\"Event.Athena.Booty.Unequipped\")")
      (Audio|PlaySoundatLocation "/Game/Sounds/Foley_Loot/Containers/Pirate_Booty/Pirate_Booty_UnEquip_Cue.Pirate_Booty_UnEquip_Cue" (Transformation|GetActorLocation _returnvalue))
      (bind _outactors (Actor|GetAllActorsOfClass "/Game/Missions/Secondary/SearchForBooty/Props/SFB_Booty.SFB_Booty_C"))
      (for _array_element _outactors
        (Class|SFBBooty|HideBeacon _array_element _returnvalue)))))

(fn ConstructionScript)
