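; Secondary mission Blueprint for Search For Booty, a treasure hunt where a player's found map leads to buried treasure.
; Mission events drive hidden use-map and dig objectives in sequence, with announcements at each stage and treasure spawning at the first.
; Completing the dig removes the collector's map and succeeds the mission.
; SpawnRewards drops pickups at the chest for players near it or the instigator, or at every player as fallback.
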(event Custom|Conversation%20Map%20Found
  ;[1 1 1] Map Found
  (bind _outfortplayerpawns (Fortnite|GetFortPlayerPawns))
  (for _array_element _outfortplayerpawns
    (if (Utilities|Equal(Object) (Fortnite|GetFortPlayerControllerfromActor _array_element) (Variables|References|GetCollecting%20Player%20Controller))
      (Utilities|MissionMacroSpawnBasicAnnouncement :TitleText (Variables|UI|GetText%20Find%20Treasure%20Convo%20Title) :DetailText (Variables|UI|GetText%20Player%20Find%20Convo%20Detail) :DisplayTime 1e+01 :RecipientControllers (Utilities|Array|MakeArray (Variables|References|GetCollecting%20Player%20Controller)))
      (else
        (Utilities|MissionMacroSpawnBasicAnnouncement :TitleText (Variables|UI|GetText%20Find%20Treasure%20Convo%20Title) :DetailText (Utilities|Text|FormatText (Variables|UI|GetText%20Other%20Players%20Convo%20Find%20Detail) (Variables|UI|GetText%20Map%20Owning%20Player%20Name)) :DisplayTime 6 :RecipientControllers (Utilities|Array|MakeArray (Fortnite|GetFortPlayerControllerfromActor _array_element)))))))

(event Custom|Conversation%20Dig%20up%20Chest
  (Utilities|MissionMacroSpawnBasicAnnouncement :TitleText (Variables|UI|GetText%20Dig%20Convo%20Title) :DetailText (Variables|UI|GetText%20Dig%20Convo%20Detail) :DisplayTime 6 :RecipientControllers (Fortnite|GetAllFortPlayerControllers true)))

(event Custom|Conversation%20Chest%20Dug%20Up
  (Utilities|MissionMacroSpawnBasicAnnouncement :TitleText (Variables|UI|GetText%20Convo%20Chest%20Title) :DetailText (Variables|UI|GetText%20Convo%20Chest%20Detail) :DisplayTime 6 :RecipientControllers (Fortnite|GetAllFortPlayerControllers true)))

(fn CompleteDigObjective
  (bind _thisobject (Class|GenTextStopObjective|SetParams (Variables|Params|GetGen%20Text%20Stop) "Succeeded"))
  (CallFunction|MissionManager|ExecuteMissionEvent(v2) "(CategoryName=\"Global\",EventName=\"StopObjective\")" (Missions|GetMissionGuid) (Variables|Params|GetTag%20Handle%20Obj%20Dig) _thisobject)
  (CallFunction|ConversationChestDugUp)
  (Inventory|RemoveItemFromPlayer :PlayerController (Variables|References|GetCollecting%20Player%20Controller) :ItemDefinition "/Game/Abilities/Player/Generic/UtilityItems/SFB_ItemDef_Map.SFB_ItemDef_Map" :AmountToRemove -1 :bForceRemoval true :bSkipVerbRemoval false)
  (CallFunction|EndMission))

(fn CompleteFindObjective
  (bind _thisobject (Class|GenTextStopObjective|SetParams (Variables|Params|GetGen%20Text%20Stop) "Succeeded"))
  (CallFunction|MissionManager|ExecuteMissionEvent(v2) "(CategoryName=\"Global\",EventName=\"StopObjective\")" (Missions|GetMissionGuid) (Variables|Params|GetTag%20Handle%20Obj%20Use%20Map) _thisobject)
  (CallFunction|StartDigObjective))

(fn EndMission
  (CallFunction|MissionManager|ExecuteMissionEvent(v2) "(CategoryName=\"Native\",EventName=\"MissionSucceeded\")" (Missions|GetMissionGuid)))

(fn SpawnRewards
  (params (InstigatorController :type Fort%20Player%20Controller%20Object%20Reference%20%28by%20ref%29) (Condition :type Boolean))
  (bind _rewardcontrollersarray (Variables|Default|GetRewardControllersArray))
  (if Condition
    ;[1 1 1] adds players who are in radius of activity and the player who initated it
    (Utilities|Array|AddUnique _rewardcontrollersarray InstigatorController)
    (for _array_element (Fortnite|GetAllFortPlayerControllers true)
      (if (<= (Transformation|GetDistanceTo _array_element (Utilities|Array|Get(acopy) _rewardcontrollersarray)) 3000)
        (Utilities|Array|AddUnique (Variables|Default|GetRewardControllersArray) _array_element)))
    (for _array_element_2 (Fortnite|GetAllFortPlayerControllers true)
      (if (<= (Transformation|GetDistanceTo _array_element_2 (Variables|References|GetSpawned%20Buried%20Treasure)) 3000)
        (Utilities|Array|AddUnique (Variables|Default|GetRewardControllersArray) _array_element_2)))
    (for _array_element_3 (Variables|Default|GetRewardControllersArray)
      (MissionHelpers|GiveMissionRewardstoPlayerasPickups (Missions|GetMissionGuid) _array_element_3 (+ (Transformation|GetActorLocation (Variables|References|GetSpawned%20Buried%20Treasure)) "0, 0,192.000000") 384 (Utilities|Struct|BreakFortMissionWeightedReward (Variables|Default|GetChosenReward)) 4))
    (else
      ;[1 1 1] fallback loot - give everyone something if it fails
      (for _array_element_4 (Fortnite|GetAllFortPlayerControllers true)
        (MissionHelpers|GiveMissionRewardstoPlayerasPickups (Missions|GetMissionGuid) _array_element_4 (+ (Transformation|GetActorLocation _array_element_4) "0, 0,25.000000") 384 (Utilities|Struct|BreakFortMissionWeightedReward (Variables|Default|GetChosenReward)) 2)))))

(fn StartDigObjective
  (bind _tag_handle_obj_dig (Variables|Params|GetTag%20Handle%20Obj%20Dig))
  (bind _thisobject (Class|GenTextStartObjective|SetParams :self (Variables|Params|GetGen%20Text%20Start) :ObjectiveText (Variables|UI|GetText%20Dig%20Objective)))
  (CallFunction|MissionManager|ExecuteMissionEvent(v2) "(CategoryName=\"Global\",EventName=\"StartObjective\")" (Missions|GetMissionGuid) _tag_handle_obj_dig _thisobject)
  (CallFunction|ConversationDigUpChest)
  (Objective|SetUIVisibilityOverride (Missions|GetObjectivebyHandle _tag_handle_obj_dig) "Invisible"))

(fn StartUseMapObjective
  (bind _tag_handle_obj_use_map (Variables|Params|GetTag%20Handle%20Obj%20Use%20Map))
  (bind _thisobject (Class|GenTextStartObjective|SetParams :self (Variables|Params|GetGen%20Text%20Start) :ObjectiveText (Utilities|Text|FormatText (Variables|UI|GetText%20Use%20Map%20Obj%20) (Variables|UI|GetText%20Map%20Owning%20Player%20Name))))
  (CallFunction|MissionManager|ExecuteMissionEvent(v2) "(CategoryName=\"Global\",EventName=\"StartObjective\")" (Missions|GetMissionGuid) _tag_handle_obj_use_map _thisobject)
  (CallFunction|SpawnBuriedTreasure)
  (CallFunction|ConversationMapFound)
  (Objective|SetUIVisibilityOverride (Missions|GetObjectivebyHandle _tag_handle_obj_use_map) "Invisible"))

(fn ConstructionScript)
