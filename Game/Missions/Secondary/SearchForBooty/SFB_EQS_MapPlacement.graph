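# EQS query for the SearchForBooty secondary mission that finds locations for map placement.
# Gathers large tiered quest-item container child actors, without limiting them to the search radius.
# Filters out candidates in water and uses a gameplay tag query to keep large quest-container locations not already marked used.
# A wide, thin, slightly raised box overlap test using complex collision on a game trace channel also checks each spot.
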EnvQuery /Game/Missions/Secondary/SearchForBooty/SFB_EQS_MapPlacement
options [{
  generator {
   name=FortQueryGenerator_ActorsAround_0
   class=FortQueryGenerator_ActorsAround
   properties {
    bAutoSortTests=true
    bCanRunAsync=false
    envQueryResultNormalizationOption=Default
    generateOnlyActorsInRadius=false
    optionName=""
    searchCenter=&EnvQueryContext_Querier
    searchRadius=1024
    searchedActorClass=/Game/Building/ActorBlueprints/Containers/Tiered_QuestItem_Large_Child.Tiered_QuestItem_Large_Child_C
   }
  }
  tests [
   FortQueryTest_InsideWater_0 {
    class=FortQueryTest_InsideWater
    properties {
     bDefineReferenceValue=false
     boolValue=false
     clampMaxType=None
     clampMinType=None
     filterType=Match
     floatValueMax=0
     floatValueMin=0
     multipleContextFilterOp=AllPass
     multipleContextScoreOp=AverageScore
     normalizationType=Absolute
     referenceValue=0
     scoreClampMax=0
     scoreClampMin=0
     scoringEquation=Constant
     scoringFactor=1
     testComment=""
     testPurpose=Filter
     testRadius=128
    }
   }
   EnvQueryTest_GameplayTags_2 {
    class=EnvQueryTest_GameplayTags
    properties {
     bDefineReferenceValue=false
     bRejectIncompatibleItems=false
     bSearchComponentsForTags=false
     boolValue=true
     clampMaxType=None
     clampMinType=None
     filterType=Match
     floatValueMax=0
     floatValueMin=0
     multipleContextFilterOp=AllPass
     multipleContextScoreOp=AverageScore
     normalizationType=Absolute
     referenceValue=0
     scoreClampMax=0
     scoreClampMin=0
     scoringEquation=Constant
     scoringFactor=1
     testComment=""
     testPurpose=Filter
     tagQueryToMatch {
      autoDescription=" ALL(  ANY( Activities.Placement.LocationQuestItemContainerLarge ),  Activities.Placement.LocationUsed ) )"
      tokenStreamVersion=0
      userDescription=""
      queryTokenStream [0,1,5,2,1,1,0,3,1,1]
      tagDictionary [{tagName=Activities.Placement.LocationQuestItemContainerLarge}, {tagName=Activities.Placement.LocationUsed}]
     }
    }
   }
   EnvQueryTest_Overlap_46 {
    class=EnvQueryTest_Overlap
    properties {
     scoringFactor=0
     overlapData {bOverlapComplex=true, extentX=192, extentY=192, extentZ=1, overlapChannel=ECC_GameTraceChannel6, overlapShape=Box, shapeOffset=<0 0 12>}
    }
   }
  ]
 }]
